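# Marker details widget for a Squad Lock, showing its title, icon, unlock hint and how many players are present out of those required.
# SetMarkerData casts the marked actor to SquadLock, tints the cached icon material via Init, then rebuilds the meeple row and count text from the lock's player counts.
# PopulateMeeple creates one SquadLockMeeple per required player, flagged present or absent.
# PreConstruct fills a mock two-of-four preview at design time.
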WidgetBlueprint /SquadLockGameplay/UI/SquadLockIndicatorDetails
graphs [
 EventGraph {nodes [
   n1 = Event PreConstruct()
   > Branch(Condition=@n1.IsDesignTime)
   > ClearChildren(self=Get Meeples_Hbox() -> (Meeples_Hbox:Horizontal Box Object Reference))
   > PopulateMeeple!(execute, self, Present:bool=true, Tint:Linear Color Structure) -> (then)
   > PopulateMeeple(Present=true)
   > PopulateMeeple(Present=false)
   > PopulateMeeple(Present=false)
   > n2 = GetDynamicMaterial(self=Get SquadLock_Image() -> (SquadLock_Image:Image Object Reference))
   > SetTextureParameterValue(self=@n2, ParameterName=DisplayedIcon, Value=Get SquadLock_Icon() -> (SquadLock_Icon:Texture Object Reference))
   > "SetText (Text)"(self=Get TextBlock_Details() -> (TextBlock_Details:Common Text Object Reference), InText="LOCTEXT(\"2 of 4 Present\")")
   > SetVectorParameterValue(self=@n2, ParameterName=FillColorInner, Value=(R=0.4902,G=0.3137,B=0.7255,A=1))
   Event Construct()
   > n3 = GetDynamicMaterial(self=Get SquadLock_Image())
   > Set SquadLock_DM(SquadLock_DM:Material Instance Dynamic Object Reference=@n3) -> (Output_Get:Material Instance Dynamic Object Reference)
   n4 = Event SetMarkerData!() -> (OutputDelegate, then, MarkerData:"Fort World Marker Data Structure (by ref)", DisplayInfo:"Marked Actor Display Info Structure (by ref)") {is=title K2Node_Event}
   > n5 = Sequence()
   > n6 = Cast To SquadLock!(
    execute
    Object=Break Fort World Marker Data!(FortWorldMarkerData:"Fort World Marker Data Structure (by ref)"=@n4.MarkerData) -> (
     MarkerID:Marker ID Structure
     Owner:Fort Player State Weak object
     MarkerType:EFortWorldMarkerType Enum
     BasePositionOffset:Vector=~
     WorldNormal:Vector=~
     ItemDefinition:Item Definition Base Object Reference
     ItemCount:int
     PinnedQuestIndex:int
     MarkedActor:Actor Weak object
     MarkedActorClass:Actor Soft Class Reference
     MarkerTag:Gameplay Tag Structure
     bUseHoveredMarkerDetail:bool=~
    ) {is=title K2Node_BreakStruct}.MarkedActor
   ) -> (then, CastFailed, AsSquad Lock:Squad Lock Object Reference) {is=title K2Node_DynamicCast}
   > Init!(
    execute
    self
    inFillColor:Linear Color Structure=Break Marked Actor Display Info!(MarkedActorDisplayInfo:"Marked Actor Display Info Structure (by ref)"=Get MarkerDisplay!(self:Building Gameplay Actor Object Reference=@n6) -> (MarkerDisplay:Marked Actor Display Info Structure)) -> (
     DisplayName:Text
     Icon:Texture 2D Soft Object Reference
     CustomIndicatorClass:User Widget Soft Class Reference
     CustomDetailsClass:User Widget Soft Class Reference
     PrimaryColor:Linear Color Structure
     Sound:Sound Base Object Reference
     bAttachSound:bool=~
     ScreenClamping:EFortMarkedActorScreenClamping Enum
     SpecialCull:Bitmask Integer
     SpecialDotSizeThreshold:float=~
     SpecialDistance:float=~
     CustomTimeToLive:float=~
     AdditionalData:Marked Actor Display Info Additional Data Object Reference
    ) {is=title K2Node_BreakStruct}.PrimaryColor
   ) -> (then)
   > n7 = Get Players In Range and Total Required To Unlock!(execute, self:Squad Lock Object Reference=@n6) -> (then, NumPlayersnRange:int=0, NumPlayersRequired:int=0, NumUnscannedPlayersInRange:int=0) {is=title K2Node_CallFunction}
   > n8 = For Loop(LastIndex=int - int(A=@n7.NumPlayersRequired))
   > PopulateMeeple(Present="integer < integer"(A=@n8, B=@n7), Tint=(R=1,G=1,B=1,A=1))
   "SetText (Text)"(
    execute=@n8.Completed
    self=Get TextBlock_Details()
    InText=Format Text!(Format="LOCTEXT(\"{x} of {y} Present\")", x:int=@n7, y:int=@n7.NumPlayersRequired) -> (Result)
   )
   "SetText (Text)"(
    execute=@n5.then_1
    self=Get ItemName() -> (ItemName:Text Object Reference)
    InText=Break Marked Actor Display Info(MarkedActorDisplayInfo=@n4.DisplayInfo) -> (bAttachSound=~, SpecialDotSizeThreshold=~, SpecialDistance=~, CustomTimeToLive=~)
   )
   Reroute Node(InputPin:Wildcard) -> (OutputPin:Wildcard)
  ]}
 Init {nodes [
   Init!() -> (then, inFillColor:Linear Color Structure) {is=title K2Node_FunctionEntry}
   > Is Valid(InputObject=Get SquadLock_DM() -> (SquadLock_DM:Material Instance Dynamic Object Reference))
   n1 = Get SquadLock_DM()
   > SetTextureParameterValue(self=@n1, ParameterName=DisplayedIcon, Value=Get SquadLock_Icon())
   > SetVectorParameterValue(self=@n1, ParameterName=FillColorInner, Value=Get inFillColor() -> (inFillColor:Linear Color Structure))
   > ClearChildren(self=Get Meeples_Hbox())
  ]}
 PopulateMeeple {nodes [
   PopulateMeeple!() -> (then, Present:bool=~, Tint:Linear Color Structure) {is=title K2Node_FunctionEntry}
   > n1 = Create Widget!(
    execute:Exec
    Class:User Widget Class Reference=&/SquadLockMeeple.SquadLockMeeple_C
    OwningPlayer:Player Controller Object Reference=GetOwningPlayer()
   ) -> (then:Exec, ReturnValue:Squad Lock Meeple Object Reference)
   > Update!(
    execute
    self:Squad Lock Meeple Object Reference=@n1
    Present:bool=Get Present() -> (Present:bool)
    Tint:Linear Color Structure=Get Tint() -> (Tint:Linear Color Structure)
   ) -> (then)
   > AddChildToHorizontalBox!(execute, self:Horizontal Box Object Reference=Get Meeples_Hbox(), Content:Widget Object Reference=@n1) -> (then, ReturnValue:Horizontal Box Slot Object Reference)
  ]}
]
variables [
 SquadLock_DM {type=MaterialInstanceDynamic}
 SquadLock_Icon {type=Texture, default=/SquadLockGameplay/UI/Icons/T-T-Icon-BR-SquadLock}
]
widgets {
 info {inheritedWidgetCount=0, parentClass=&UserWidget, rootWidgetClass=&HorizontalBox, widgetCount=13}
 widgets [
  HorizontalBox_67 {
   widgetClassPath=&HorizontalBox
   widget=HorizontalBox_67
   properties {renderTransform {angle=0, scale=<1 1>, shear=<0 0>, translation=<-50 50>}}
  }
  Overlay_1 {
   widgetClassPath=&Overlay
   parent=HorizontalBox_67
   widget=Overlay_1
   slot {class=HorizontalBoxSlot, properties {horizontalAlignment=HAlign_Left, verticalAlignment=VAlign_Top, padding=<ltrb 0 0 -6 0>}}
  }
  SquadLock_Image {
   widgetClassPath=&Image
   bIsVariable=true
   parent=Overlay_1
   widget=SquadLock_Image
   properties {brush {
     bIsDynamicallyLoaded=false
     drawAs=Image
     imageType=NoImage
     mirroring=NoMirror
     resourceName=None
     resourceObject=/SquadLockGameplay/UI/Icons/Map/T_UI_SquadLock_Map_64
     tiling=NoTile
     imageSize=<64 64>
     margin=<ltrb 0 0 0 0>
     outlineSettings {
      bUseBrushTransparency=false
      roundingType=HalfHeightRadius
      width=0
      color {colorUseRule=UseColor_Specified, specifiedColor=<rgba 0 0 0 0>}
      cornerRadii=<0 0 0 0>
     }
     tintColor {colorUseRule=UseColor_Specified, specifiedColor=<rgba 1 1 1 1>}
     uVRegion {bIsValid=false, max=<0 0>, min=<0 0>}
    }}
   slot {class=OverlaySlot, properties {horizontalAlignment=HAlign_Center, verticalAlignment=VAlign_Center}}
  }
  VerticalBox_0 {
   widgetClassPath=&VerticalBox
   parent=HorizontalBox_67
   widget=VerticalBox_0
   slot {class=HorizontalBoxSlot, properties {horizontalAlignment=HAlign_Left, verticalAlignment=VAlign_Top}}
  }
  HorizontalBox_0 {
   widgetClassPath=&HorizontalBox
   parent=VerticalBox_0
   widget=HorizontalBox_0
   slot {class=VerticalBoxSlot, properties {horizontalAlignment=HAlign_Left, padding=<ltrb 6 0 8 -6>}}
  }
  ItemName {
   widgetClassPath=&TextBlock
   bIsVariable=true
   parent=HorizontalBox_0
   widget=ItemName
   properties {
    text="{MARKER TITLE}"
    font {
     bForceMonospaced=false
     bMaterialIsStencil=false
     fontMaterial=null
     fontObject=&Burbank
     letterSpacing=0
     monospacedWidth=1
     size=20
     skewAmount=0
     typefaceFontName=Burbank-PlayerPower
     outlineSettings {
      bApplyOutlineToDropShadows=true
      bMiteredCorners=false
      bSeparateFillAlpha=true
      outlineMaterial=&M-FontShadow
      outlineSize=3
      outlineColor=<rgba 0 0 0 0.8>
     }
    }
    renderTransform {angle=0, scale=<1 1>, shear=<-8 0>, translation=<0 0>}
   }
   slot {class=HorizontalBoxSlot, properties {horizontalAlignment=HAlign_Left, verticalAlignment=VAlign_Bottom, padding=<ltrb 8 -5 0 -5>}}
  }
  Overlay {widgetClassPath=&Overlay, parent=VerticalBox_0, widget=Overlay, slot {class=VerticalBoxSlot, properties {padding=<ltrb 0 4 0 0>}}}
  Seperator {
   widgetClassPath=&Border
   bIsVariable=true
   parent=Overlay
   widget=Seperator
   properties {
    background {
     bIsDynamicallyLoaded=false
     drawAs=Image
     imageType=NoImage
     mirroring=NoMirror
     resourceName=None
     resourceObject=null
     tiling=NoTile
     imageSize=<32 16>
     margin=<ltrb 0 0 0 0>
     outlineSettings {
      bUseBrushTransparency=false
      roundingType=HalfHeightRadius
      width=0
      color {colorUseRule=UseColor_Specified, specifiedColor=<rgba 0 0 0 0>}
      cornerRadii=<0 0 0 0>
     }
     tintColor {colorUseRule=UseColor_Specified, specifiedColor=<rgba 1 1 1 1>}
     uVRegion {bIsValid=false, max=<0 0>, min=<0 0>}
    }
    brushColor=<rgba 0 0.4059 0.875 1>
    contentColorAndOpacity=<rgba 0 0.4059 0.875 1>
    padding=<ltrb 4 0 4 3>
    renderTransformPivot=<0 0.5>
   }
   slot {class=OverlaySlot, properties {horizontalAlignment=HAlign_Fill, verticalAlignment=VAlign_Center, padding=<ltrb 3 0 0 0>}}
  }
  Hint_Text {
   widgetClassPath=&TextBlock
   parent=VerticalBox_0
   widget=Hint_Text
   properties {
    text="Gather players here to unlock!"
    font {
     bForceMonospaced=false
     bMaterialIsStencil=false
     fontMaterial=null
     fontObject=&BurbankSmall
     letterSpacing=0
     monospacedWidth=1
     size=18
     skewAmount=0
     typefaceFontName=Bold
     outlineSettings {
      bApplyOutlineToDropShadows=true
      bMiteredCorners=false
      bSeparateFillAlpha=true
      outlineMaterial=&M-FontShadow
      outlineSize=3
      outlineColor=<rgba 0 0 0 0.8>
     }
    }
    renderTransform {angle=0, scale=<1 1>, shear=<-8 0>, translation=<0 0>}
   }
   slot {class=VerticalBoxSlot, properties {horizontalAlignment=HAlign_Left, verticalAlignment=VAlign_Bottom, padding=<ltrb 8 0 0 -5>}}
  }
  SizeBox_1 {
   widgetClassPath=&SizeBox
   parent=VerticalBox_0
   widget=SizeBox_1
   properties {bOverride_HeightOverride=true, heightOverride=42}
   slot {class=VerticalBoxSlot, properties {horizontalAlignment=HAlign_Left, padding=<ltrb 8 6 0 0>}}
  }
  ScaleBox_0 {
   widgetClassPath=&ScaleBox
   parent=SizeBox_1
   widget=ScaleBox_0
   slot {class=SizeBoxSlot, properties {horizontalAlignment=HAlign_Left}}
  }
  Meeples_Hbox {
   widgetClassPath=&HorizontalBox
   bIsVariable=true
   parent=ScaleBox_0
   widget=Meeples_Hbox
   slot {class=ScaleBoxSlot, properties {horizontalAlignment=HAlign_Left}}
  }
  TextBlock_Details {
   widgetClassPath=&CommonTextBlock
   bIsVariable=true
   parent=VerticalBox_0
   widget=TextBlock_Details
   properties {
    font {
     bForceMonospaced=false
     bMaterialIsStencil=false
     fontMaterial=null
     fontObject=&BurbankSmall
     letterSpacing=0
     monospacedWidth=1
     size=14
     skewAmount=0
     typefaceFontName=Black
     outlineSettings {
      bApplyOutlineToDropShadows=true
      bMiteredCorners=false
      bSeparateFillAlpha=true
      outlineMaterial=&M-FontShadow
      outlineSize=3
      outlineColor=<rgba 0 0 0 0.8>
     }
    }
    shadowOffset=<0 0>
   }
   slot {class=VerticalBoxSlot, properties {horizontalAlignment=HAlign_Left, verticalAlignment=VAlign_Top, padding=<ltrb 12 6 0 0>}}
  }
 ]
}
