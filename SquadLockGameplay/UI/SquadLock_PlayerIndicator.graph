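# SquadLock player indicator widget: a scaled icon that shows a player's in-range and scan state.
# Update UI Player In Range State takes two flags; a pending scan loops Anim_BeingScanned, an opacity blink, with the icon tinted blue.
# Otherwise the icon turns red when in range and scanned or blue when not, and any playing blink stops.
# The event graph holds only empty event stubs, so updates rely on that function being called.
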WidgetBlueprint /SquadLockGameplay/UI/SquadLock_PlayerIndicator
animations [Anim_BeingScanned {
  binding_count=1
  length_seconds=0.5
  object=&self:Anim_BeingScanned
  sequence {bindings [PlayerIcon {
     id=#1
     possessed_class=Image
     tracks [{
       type=MovieSceneFloatTrack
       display_name=Render Opacity
       sections [{
         type=MovieSceneFloatSection
         end_frame=10
         start_frame=0
         channels [None {type=MovieSceneScriptingFloatChannel, keys [[0,1],[3,0],[7,0],[10,1]]}]
        }]
      }]
    }]}
 }]
graphs [
 EventGraph {nodes [
   Event PreConstruct()
   Event Construct()
   Event Tick()
  ]}
 Update UI Player In Range State {nodes [
   n1 = Update UI Player In Range State!() -> (then, bInRangeAndScanned:bool=~, bInRangePendingScan:bool=~) {is=title K2Node_FunctionEntry}
   > n2 = Branch(Condition=@n1.bInRangePendingScan)
   > PlayAnimation(InAnimation=Get Anim_BeingScanned() -> (Anim_BeingScanned:Widget Animation Object Reference), NumLoopsToPlay=0)
   > SetColorAndOpacity(
    self:Image Object Reference=Get PlayerIcon() -> (PlayerIcon:Image Object Reference)
    InColorAndOpacity:Linear Color Structure=(R=0.0648,G=0.3278,B=1,A=1)
   )
   SetColorAndOpacity(execute=@n2.else, self=Get PlayerIcon(), InColorAndOpacity=SelectColor(A=(R=1,G=0,B=0,A=1), B=(R=0.0648,G=0.3278,B=1,A=1), bPickA=@n1))
   n3 = Get Anim_BeingScanned()
   > Branch(Condition=IsAnimationPlaying(InAnimation=@n3))
   > StopAnimation(InAnimation=@n3)
   SetColorAndOpacity(InColorAndOpacity=(R=1,G=0.9237,B=0,A=1))
  ]}
]
variables []
widgets {
 info {inheritedWidgetCount=0, parentClass=&UserWidget, rootWidgetClass=&ScaleBox, widgetCount=2}
 widgets [
  ScaleBox_0 {widgetClassPath=&ScaleBox, widget=ScaleBox_0, properties {stretch=UserSpecified, userSpecifiedScale=1.2}}
  PlayerIcon {
   widgetClassPath=&Image
   bIsVariable=true
   parent=ScaleBox_0
   widget=PlayerIcon
   properties {
    visibility=HitTestInvisible
    brush {
     bIsDynamicallyLoaded=false
     drawAs=Image
     imageType=NoImage
     mirroring=NoMirror
     resourceName=None
     resourceObject=/SquadLockGameplay/UI/Materials/M_UI_SquadLockMember_Inst
     tiling=NoTile
     imageSize=<100 100>
     margin=<ltrb 0 0 0 0>
     outlineSettings {
      bUseBrushTransparency=false
      roundingType=HalfHeightRadius
      width=0
      color {colorUseRule=UseColor_Specified, specifiedColor=<rgba 0 0 0 0>}
      cornerRadii=<0 0 0 0>
     }
     tintColor {colorUseRule=UseColor_Specified, specifiedColor=<rgba 1 1 1 1>}
     uVRegion {bIsValid=false, max=<0 0>, min=<0 0>}
    }
   }
   slot {class=ScaleBoxSlot, properties {horizontalAlignment=HAlign_Fill, verticalAlignment=VAlign_Fill}}
  }
 ]
}
