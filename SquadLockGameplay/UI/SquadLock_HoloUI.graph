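# Holographic screen widget for a squad lock, showing player indicators, access status and a padlock.
# Custom events switch between scanning, granted, denied and target-lost views, with overridable text and a flashing target-lost message.
# Count functions rebuild the indicator row and update each indicator's scanned or pending state.
# Unlocking tints the fade green and plays an animation that turns the padlock green, opens it and shrinks the widget away.
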WidgetBlueprint /SquadLockGameplay/UI/SquadLock_HoloUI
animations [
 FlashText {
  binding_count=1
  length_seconds=0.7758
  object=&self:FlashText
  sequence {
   bindings [CommonBorder_40 {
     id=#1
     possessed_class=CommonBorder
     tracks [{
       type=MovieSceneFloatTrack
       display_name=Render Opacity
       sections [{
         type=MovieSceneFloatSection
         channels [None {type=MovieSceneScriptingFloatChannel, keys [[0,1],[2.25,0],[4.5,1],[6.75,0],[9,1],[11.25,0],[13.5,1],[15.516,1]]}]
        }]
      }]
    }]
   tracks [{type=MovieSceneAudioTrack, display_name=Audio, sections []}]
  }
 }
 Anim_FlashUI {
  binding_count=1
  length_seconds=1
  object=&self:Anim_FlashUI
  sequence {bindings [WrapBox_98 {
     id=#2
     possessed_class=WrapBox
     tracks [{
       type=MovieSceneFloatTrack
       display_name=Render Opacity
       sections [{
         type=MovieSceneFloatSection
         end_frame=20
         start_frame=0
         channels [None {type=MovieSceneScriptingFloatChannel, keys [[0,0.8],[5,0],[7,0],[12,0.8]]}]
        }]
      }]
    }]}
 }
 Anim_ShowPadlock {
  binding_count=1
  length_seconds=1.5
  object=&self:Anim_ShowPadlock
  sequence {bindings [Border_PadlockDisplay {
     id=#3
     possessed_class=Border
     tracks [
      {
       type=MovieSceneFloatTrack
       display_name=Render Opacity
       sections [{
         type=MovieSceneFloatSection
         end_frame=30
         start_frame=0
         channels [None {type=MovieSceneScriptingFloatChannel, keys [[0,0],[20,0],[25,0.8]]}]
        }]
      }
      {
       type=MovieSceneColorTrack
       display_name=Content Color and Opacity
       sections [{
         type=MovieSceneColorSection
         end_frame=30
         start_frame=0
         channels [
          Color.R {type=MovieSceneScriptingFloatChannel, const [1,2]}
          Color.G {type=MovieSceneScriptingFloatChannel, const [0,2]}
          Color.B {type=MovieSceneScriptingFloatChannel, const [0,2]}
          Color.A {type=MovieSceneScriptingFloatChannel, const [1,2]}
         ]
        }]
      }
     ]
    }]}
 }
 Anim_UnlockFinal {
  binding_count=3
  length_seconds=3.7
  object=&self:Anim_UnlockFinal
  sequence {bindings [
    SquadLock_HoloUI_C_82 {
     id=#4
     possessed_class=SquadLock_HoloUI_C
     tracks [
      {
       type=MovieScene2DTransformTrack
       sections [{
         type=MovieScene2DTransformSection
         end_frame=74
         start_frame=64
         channels [
          Translation.X {type=MovieSceneScriptingFloatChannel, const [0,1]}
          Translation.Y {type=MovieSceneScriptingFloatChannel, const [0,1]}
          Angle {type=MovieSceneScriptingFloatChannel, const [0,1]}
          Scale.X {type=MovieSceneScriptingFloatChannel, keys [[64,1],[68,1.176],[74,0]]}
          Scale.Y {type=MovieSceneScriptingFloatChannel, keys [[64,1],[68,1.176],[74,0]]}
          Shear.X {type=MovieSceneScriptingFloatChannel, const [0,1]}
          Shear.Y {type=MovieSceneScriptingFloatChannel, const [0,1]}
         ]
        }]
      }
      {
       type=MovieSceneFloatTrack
       display_name=Render Opacity
       sections [{type=MovieSceneFloatSection, end_frame=74, start_frame=64, channels [None {type=MovieSceneScriptingFloatChannel, keys [[64,1],[74,0]]}]}]
      }
     ]
    }
    Border_PadlockDisplay {
     id=#3
     possessed_class=Border
     tracks [
      {
       type=MovieSceneColorTrack
       display_name=Content Color and Opacity
       sections [{
         type=MovieSceneColorSection
         end_frame=14
         start_frame=0
         channels [
          Color.R {type=MovieSceneScriptingFloatChannel, keys [[-6,1],[4.0003,0]]}
          Color.G {type=MovieSceneScriptingFloatChannel, keys [[-6,0],[4.0003,1]]}
          Color.B {type=MovieSceneScriptingFloatChannel, const [0,2]}
          Color.A {type=MovieSceneScriptingFloatChannel, const [1,2]}
         ]
        }]
      }
      {
       type=MovieSceneFloatTrack
       display_name=Render Opacity
       sections [{
         type=MovieSceneFloatSection
         end_frame=27
         start_frame=0
         channels [None {type=MovieSceneScriptingFloatChannel, keys [[-16,0],[-11,0],[-6,0.8]]}]
        }]
      }
     ]
    }
    Lock_Top {
     id=#5
     possessed_class=Image
     tracks [{
       type=MovieScene2DTransformTrack
       sections [{
         type=MovieScene2DTransformSection
         end_frame=24
         start_frame=0
         channels [
          Translation.X {type=MovieSceneScriptingFloatChannel, const [0,1]}
          Translation.Y {type=MovieSceneScriptingFloatChannel, keys [[-16,10],[4.0003,10],[14.0007,-10]]}
          Angle {type=MovieSceneScriptingFloatChannel, keys [[-16,0],[4.0003,0],[14.0007,-20]]}
          Scale.X {type=MovieSceneScriptingFloatChannel, const [1,1]}
          Scale.Y {type=MovieSceneScriptingFloatChannel, const [1,1]}
          Shear.X {type=MovieSceneScriptingFloatChannel, const [0,1]}
          Shear.Y {type=MovieSceneScriptingFloatChannel, const [0,1]}
         ]
        }]
      }]
    }
   ]}
 }
]
graphs [
 EventGraph {nodes [
   n1 = EventTargetLost!() -> (OutputDelegate, then, ScreenText:Text) {is=title K2Node_CustomEvent}
   > "SetText (Text)"(
    self=Get Text_Misc() -> (Text_Misc:Common Text Object Reference)
    InText=Select(Option 0:Text=@n1.ScreenText, Option 1:Text="LOCTEXT(\"Target\\r\\nLost\")", Index=TextIsEmpty(InText=@n1.ScreenText)) -> (ReturnValue:Text)
   )
   n2 = Get CommonWidgetSwitcher_States() -> (CommonWidgetSwitcher_States:Common Widget Switcher Legacy Object Reference)
   > SetActiveWidgetIndex(self=@n2, Index=3)
   > n3 = PlayAnimation(InAnimation=Get FlashText() -> (FlashText:Widget Animation Object Reference))
   n4 = Get Overlay_BGIdle() -> (Overlay_BGIdle:Overlay Object Reference)
   SetVisibility(execute=[@n7.then, @n3.then], self=@n4, InVisibility=Visible)
   n5 = EventFail!() -> (OutputDelegate, then, ScreenText:Text) {is=title K2Node_CustomEvent}
   > "SetText (Text)"(
    self=Get Text_Failure() -> (Text_Failure:Common Text Object Reference)
    InText=Select(Option 0=@n5.ScreenText, Option 1="LOCTEXT(\"Access\\r\\nDenied\")", Index=TextIsEmpty(InText=@n5.ScreenText))
   )
   > n6 = SetActiveWidgetIndex(self=@n2, Index=2)
   SetVisibility(execute=[@n6.then, @n9.then], self=@n4, InVisibility=Hidden)
   EventScanning() {is=title K2Node_CustomEvent}
   > n7 = SetActiveWidgetIndex(self=@n2, Index=0)
   n8 = EventSuccess!() -> (OutputDelegate, then, ScreenText:Text) {is=title K2Node_CustomEvent}
   > "SetText (Text)"(
    self=Get Text_Success() -> (Text_Success:Common Text Object Reference)
    InText=Select(Option 0=@n8.ScreenText, Option 1="LOCTEXT(\"Access\\r\\nGranted\")", Index=TextIsEmpty(InText=@n8.ScreenText))
   )
   > n9 = SetActiveWidgetIndex(self=@n2, Index=1)
   PlayUnlockAnim() {is=title K2Node_CustomEvent}
   > PlayAnimation(InAnimation=Get Anim_UnlockFinal() -> (Anim_UnlockFinal:Widget Animation Object Reference))
   ShowPadlock() {is=title K2Node_CustomEvent}
   > SetVisibility(self=Get Border_PadlockDisplay() -> (Border_PadlockDisplay:Border Object Reference), InVisibility=HitTestInvisible)
   > PlayAnimation(InAnimation=Get Anim_ShowPadlock() -> (Anim_ShowPadlock:Widget Animation Object Reference))
   n10 = SetMeepleVisibility!() -> (OutputDelegate, then, bVisible:bool=~) {is=title K2Node_CustomEvent}
   > SetVisibility(
    self=Get WrapBox_98() -> (WrapBox_98:Wrap Box Object Reference)
    InVisibility=Select(Option 0:ESlateVisibility Enum=Hidden, Option 1:ESlateVisibility Enum=HitTestInvisible, Index=@n10.bVisible) -> (ReturnValue:ESlateVisibility Enum)
   )
   "Play Animation with Finished event (legacy)"!(
    execute:Exec
    Widget:User Widget Object Reference=Self-Reference()
    InAnimation:Widget Animation Object Reference=Get Anim_FlashUI() -> (Anim_FlashUI:Widget Animation Object Reference)
    StartAtTime:float=0
    NumLoopsToPlay:int=1
    PlayMode:EUMGSequencePlayMode Enum
    PlaybackSpeed:float=1
   ) -> (then:Exec, Result:UMGSequence Player Object Reference, Finished:Exec) {is=title K2Node_PlayAnimation}
  ]}
 Sequencer Events {links [], nodes []}
 UpdateCurrentPlayerCount {nodes [
   UpdateCurrentPlayerCount!() -> (then, NumPlayers:int, NumPendingPlayers:int) {is=title K2Node_FunctionEntry}
   > Set LocalPendingPlayerCount(LocalPendingPlayerCount:int=0) -> (Output_Get:int=0)
   n1 = Get WrapBox_98()
   > n2 = For Loop(LastIndex=int - int(A=GetChildrenCount!(self:Panel Widget Object Reference=@n1) -> (ReturnValue:int=0), B=1) {is=title K2Node_CallFunction})
   > Set LocalPlayerIsPending(LocalPlayerIsPending:bool=false) -> (Output_Get:bool)
   > n3 = Cast To SquadLock_PlayerIndicator!(execute, Object=GetChildAt!(self:Panel Widget Object Reference=Get WrapBox_98(), Index:int=@n2) -> (ReturnValue:Widget Object Reference)) -> (then, CastFailed, AsSquad Lock Player Indicator:Squad Lock Player Indicator Object Reference) {is=title K2Node_DynamicCast}
   n4 = "integer >= integer"!(A:int=@n2, B:int=Get NumPlayers() -> (NumPlayers:int=0)) -> (ReturnValue:bool) {is=title K2Node_CallFunction}
   > n5 = Branch(Condition=@n4)
   > n6 = Branch(Condition="integer >= integer"(A=Get LocalPendingPlayerCount() -> (LocalPendingPlayerCount:int=0), B=Get NumPendingPlayers() -> (NumPendingPlayers:int=0)) {is=title K2Node_PromotableOperator})
   > n7 = Branch(Condition=Get bIsUnlocked() -> (bIsUnlocked:bool))
   Update UI Player In Range State!(
    execute=[@n7.else, @n5.else, @n8."  "]
    self:Squad Lock Player Indicator Object Reference=@n3
    bInRangeAndScanned:bool=@n4
    bInRangePendingScan:bool=Get LocalPlayerIsPending() -> (LocalPlayerIsPending:bool)
   ) -> (then)
   Set LocalPlayerIsPending(execute=@n6.else, LocalPlayerIsPending=true)
   > n8 = Increment Int(Value=Get LocalPendingPlayerCount() -> (LocalPendingPlayerCount=0))
  ]}
 UpdateMaxPlayerCount {nodes [
   n1 = UpdateMaxPlayerCount!() -> (then, CountNeeded:int) {is=title K2Node_FunctionEntry}
   > ClearChildren(self=Get WrapBox_98())
   > For Loop(LastIndex=int - int(A=@n1, B=1))
   > n2 = Create Widget!(
    execute:Exec
    Class:User Widget Class Reference=&/SquadLock_PlayerIndicator.SquadLock_PlayerIndicator_C
    OwningPlayer:Player Controller Object Reference
   ) -> (then:Exec, ReturnValue:Squad Lock Player Indicator Object Reference)
   > AddChildToWrapBox!(execute, self:Wrap Box Object Reference=Get WrapBox_98(), Content:Widget Object Reference=@n2) -> (then, ReturnValue:Wrap Box Slot Object Reference)
  ]}
 UpdateUnlockState {nodes [
   n1 = UpdateUnlockState!() -> (then, bInlocked:bool=~) {is=title K2Node_FunctionEntry}
   > n2 = Set bIsUnlocked(bIsUnlocked:bool=@n1) -> (Output_Get:bool)
   > SetColorAndOpacity(
    self:Image Object Reference=Get Image_HoloFade() -> (Image_HoloFade:Image Object Reference)
    InColorAndOpacity:Linear Color Structure=SelectColor(A=(R=0.0033,G=1,B=0,A=1), B=(R=1,G=0,B=0,A=1), bPickA=@n2)
   )
   > Branch(Condition=Get bIsUnlocked())
   > PlayUnlockAnim() {is=title K2Node_CallFunction}
  ]}
]
variables [
 bIsUnlocked {type=boolean}
 NumFlashes {type=integer}
]
widgets {
 info {inheritedWidgetCount=0, parentClass=&UserWidget, rootWidgetClass=&Overlay, widgetCount=24}
 widgets [
  Overlay_0 {widgetClassPath=&Overlay, widget=Overlay_0}
  CommonBorder_42 {
   widgetClassPath=&CommonBorder
   parent=Overlay_0
   widget=CommonBorder_42
   properties {visibility=Visible, padding=<ltrb 0 0 0 0>}
   slot {class=OverlaySlot, properties {horizontalAlignment=HAlign_Center, verticalAlignment=VAlign_Center}}
  }
  SizeBox_0 {
   widgetClassPath=&SizeBox
   parent=CommonBorder_42
   widget=SizeBox_0
   properties {bOverride_HeightOverride=true, bOverride_WidthOverride=true, heightOverride=120, widthOverride=200}
   slot {class=BorderSlot, properties {padding=<ltrb 0 0 0 0>}}
  }
  ScaleBox_0 {widgetClassPath=&ScaleBox, parent=SizeBox_0, widget=ScaleBox_0, slot {class=SizeBoxSlot}}
  Overlay_1 {
   widgetClassPath=&Overlay
   parent=ScaleBox_0
   widget=Overlay_1
   slot {class=ScaleBoxSlot, properties {horizontalAlignment=HAlign_Fill, verticalAlignment=VAlign_Fill}}
  }
  Overlay_BGIdle {
   widgetClassPath=&Overlay
   bIsVariable=true
   parent=Overlay_1
   widget=Overlay_BGIdle
   slot {class=OverlaySlot, properties {horizontalAlignment=HAlign_Fill, verticalAlignment=VAlign_Fill}}
  }
  Image_HoloFade {
   widgetClassPath=&Image
   bIsVariable=true
   parent=Overlay_BGIdle
   widget=Image_HoloFade
   properties {
    renderOpacity=0.25
    brush {
     bIsDynamicallyLoaded=false
     drawAs=Image
     imageType=NoImage
     mirroring=NoMirror
     resourceName=None
     resourceObject=/SquadLockGameplay/UI/Materials/t_HoloFade
     tiling=NoTile
     imageSize=<512 150>
     margin=<ltrb 0 0 0 0>
     outlineSettings {
      bUseBrushTransparency=false
      roundingType=HalfHeightRadius
      width=0
      color {colorUseRule=UseColor_Specified, specifiedColor=<rgba 0 0 0 0>}
      cornerRadii=<0 0 0 0>
     }
     tintColor {colorUseRule=UseColor_Specified, specifiedColor=<rgba 1 1 1 1>}
     uVRegion {bIsValid=false, max=<0 0>, min=<0 0>}
    }
    colorAndOpacity=<rgba 1 0 0 1>
   }
   slot {class=OverlaySlot, properties {horizontalAlignment=HAlign_Fill, verticalAlignment=VAlign_Bottom, padding=<ltrb 0 200 0 0>}}
  }
  Image_ScreenBorder {
   widgetClassPath=&Image
   bIsVariable=true
   parent=Overlay_1
   widget=Image_ScreenBorder
   properties {
    visibility=Collapsed
    brush {
     bIsDynamicallyLoaded=false
     drawAs=NoDrawType
     imageType=NoImage
     mirroring=NoMirror
     resourceName=None
     resourceObject=/Game/Athena/HUD/Vehicles/Materials/M_UI_TargetBracket_Inst
     tiling=NoTile
     imageSize=<64 64>
     margin=<ltrb 0.4 0.4 0.4 0.4>
     outlineSettings {
      bUseBrushTransparency=false
      roundingType=HalfHeightRadius
      width=0
      color {colorUseRule=UseColor_Specified, specifiedColor=<rgba 0 0 0 0>}
      cornerRadii=<0 0 0 0>
     }
     tintColor {colorUseRule=UseColor_Specified, specifiedColor=<rgba 1 1 1 1>}
     uVRegion {bIsValid=false, max=<0 0>, min=<0 0>}
    }
   }
   slot {class=OverlaySlot, properties {horizontalAlignment=HAlign_Fill, verticalAlignment=VAlign_Fill, padding=<ltrb 10 10 10 10>}}
  }
  CommonBorder_40 {
   widgetClassPath=&CommonBorder
   parent=Overlay_1
   widget=CommonBorder_40
   properties {verticalAlignment=VAlign_Bottom, padding=<ltrb 0 0 0 0>}
   slot {class=OverlaySlot, properties {horizontalAlignment=HAlign_Fill, verticalAlignment=VAlign_Fill, padding=<ltrb 20 20 20 20>}}
  }
  ScaleBox_1 {
   widgetClassPath=&ScaleBox
   parent=CommonBorder_40
   widget=ScaleBox_1
   properties {stretch=None}
   slot {class=BorderSlot, properties {verticalAlignment=VAlign_Bottom, padding=<ltrb 0 0 0 0>}}
  }
  CommonWidgetSwitcher_States {
   widgetClassPath=&CommonWidgetSwitcherLegacy
   bIsVariable=true
   parent=ScaleBox_1
   widget=CommonWidgetSwitcher_States
   properties {transitionDuration=0, transitionType=Zoom}
   slot {class=ScaleBoxSlot, properties {horizontalAlignment=HAlign_Fill}}
  }
  Overlay_2 {
   widgetClassPath=&Overlay
   parent=CommonWidgetSwitcher_States
   widget=Overlay_2
   slot {class=WidgetSwitcherSlot, properties {verticalAlignment=VAlign_Bottom}}
  }
  CommonTextBlock_1 {
   widgetClassPath=&CommonTextBlock
   parent=Overlay_2
   widget=CommonTextBlock_1
   properties {
    justification=Center
    text=BRING MORE
    textTransformPolicy=ToUpper
    visibility=Collapsed
    font {
     bForceMonospaced=false
     bMaterialIsStencil=false
     fontMaterial=null
     fontObject=&Burbank
     letterSpacing=0
     monospacedWidth=1
     size=18
     skewAmount=0
     typefaceFontName=BurbankRegular-Black
     outlineSettings {
      bApplyOutlineToDropShadows=false
      bMiteredCorners=false
      bSeparateFillAlpha=false
      outlineMaterial=null
      outlineSize=0
      outlineColor=<rgba 0 0 0 1>
     }
    }
    shadowOffset=<0 0>
   }
   slot {class=OverlaySlot, properties {verticalAlignment=VAlign_Center}}
  }
  WrapBox_98 {
   widgetClassPath=&WrapBox
   bIsVariable=true
   parent=Overlay_2
   widget=WrapBox_98
   properties {bExplicitWrapSize=true, horizontalAlignment=HAlign_Center, renderOpacity=0.8, wrapSize=1200, innerSlotPadding=<-25 -10>}
   slot {class=OverlaySlot, properties {horizontalAlignment=HAlign_Fill, verticalAlignment=VAlign_Center, padding=<ltrb 80 0 80 25>}}
  }
  SquadLock_PlayerIndicator {
   widgetClassPath=&/SquadLock_PlayerIndicator
   bIsVariable=true
   parent=WrapBox_98
   widget=SquadLock_PlayerIndicator
   slot {class=WrapBoxSlot}
  }
  SquadLock_PlayerIndicator_2814 {
   widgetClassPath=&SquadLock_PlayerIndicator
   bIsVariable=true
   parent=WrapBox_98
   widget=SquadLock_PlayerIndicator_2814
   slot {class=WrapBoxSlot}
  }
  SquadLock_PlayerIndicator_2917 {
   widgetClassPath=&SquadLock_PlayerIndicator
   bIsVariable=true
   parent=WrapBox_98
   widget=SquadLock_PlayerIndicator_2917
   slot {class=WrapBoxSlot}
  }
  Text_Success {
   widgetClassPath=&CommonTextBlock
   bIsVariable=true
   parent=CommonWidgetSwitcher_States
   widget=Text_Success
   properties {
    justification=Center
    lineHeightPercentage=0.8
    text="Access\r\nGranted"
    textTransformPolicy=ToUpper
    font {
     bForceMonospaced=false
     bMaterialIsStencil=false
     fontMaterial=null
     fontObject=&Burbank
     letterSpacing=0
     monospacedWidth=1
     size=18
     skewAmount=0
     typefaceFontName=BurbankRegular-Black
     outlineSettings {
      bApplyOutlineToDropShadows=false
      bMiteredCorners=false
      bSeparateFillAlpha=false
      outlineMaterial=null
      outlineSize=0
      outlineColor=<rgba 0 0 0 1>
     }
    }
    shadowOffset=<0 0>
   }
   slot {class=WidgetSwitcherSlot, properties {horizontalAlignment=HAlign_Center, verticalAlignment=VAlign_Center}}
  }
  Text_Failure {
   widgetClassPath=&CommonTextBlock
   bIsVariable=true
   parent=CommonWidgetSwitcher_States
   widget=Text_Failure
   properties {
    justification=Center
    lineHeightPercentage=0.8
    text="Access\r\nDenied"
    textTransformPolicy=ToUpper
    font {
     bForceMonospaced=false
     bMaterialIsStencil=false
     fontMaterial=null
     fontObject=&Burbank
     letterSpacing=0
     monospacedWidth=1
     size=18
     skewAmount=0
     typefaceFontName=BurbankRegular-Black
     outlineSettings {
      bApplyOutlineToDropShadows=false
      bMiteredCorners=false
      bSeparateFillAlpha=false
      outlineMaterial=null
      outlineSize=0
      outlineColor=<rgba 0 0 0 1>
     }
    }
    shadowOffset=<0 0>
   }
   slot {class=WidgetSwitcherSlot, properties {horizontalAlignment=HAlign_Center, verticalAlignment=VAlign_Center}}
  }
  Text_Misc {
   widgetClassPath=&CommonTextBlock
   bIsVariable=true
   parent=CommonWidgetSwitcher_States
   widget=Text_Misc
   properties {
    justification=Center
    lineHeightPercentage=0.8
    text="Target\r\nLost"
    textTransformPolicy=ToUpper
    font {
     bForceMonospaced=false
     bMaterialIsStencil=false
     fontMaterial=null
     fontObject=&Burbank
     letterSpacing=0
     monospacedWidth=1
     size=18
     skewAmount=0
     typefaceFontName=BurbankRegular-Black
     outlineSettings {
      bApplyOutlineToDropShadows=false
      bMiteredCorners=false
      bSeparateFillAlpha=false
      outlineMaterial=null
      outlineSize=0
      outlineColor=<rgba 0 0 0 1>
     }
    }
    shadowOffset=<0 0>
   }
   slot {class=WidgetSwitcherSlot, properties {horizontalAlignment=HAlign_Center, verticalAlignment=VAlign_Center}}
  }
  Border_PadlockDisplay {
   widgetClassPath=&Border
   bIsVariable=true
   parent=Overlay_1
   widget=Border_PadlockDisplay
   properties {
    renderOpacity=0
    visibility=Collapsed
    background {
     bIsDynamicallyLoaded=false
     drawAs=NoDrawType
     imageType=NoImage
     mirroring=NoMirror
     resourceName=None
     resourceObject=null
     tiling=NoTile
     imageSize=<32 32>
     margin=<ltrb 0 0 0 0>
     outlineSettings {
      bUseBrushTransparency=false
      roundingType=HalfHeightRadius
      width=0
      color {colorUseRule=UseColor_Specified, specifiedColor=<rgba 0 0 0 0>}
      cornerRadii=<0 0 0 0>
     }
     tintColor {colorUseRule=UseColor_Specified, specifiedColor=<rgba 1 1 1 1>}
     uVRegion {bIsValid=false, max=<0 0>, min=<0 0>}
    }
    padding=<ltrb 0 0 0 0>
    renderTransform {angle=0, scale=<1 1>, shear=<0 0>, translation=<0 30>}
   }
   slot {class=OverlaySlot, properties {horizontalAlignment=HAlign_Center, verticalAlignment=VAlign_Center}}
  }
  Overlay_3 {
   widgetClassPath=&Overlay
   parent=Border_PadlockDisplay
   widget=Overlay_3
   slot {class=BorderSlot, properties {padding=<ltrb 0 0 0 0>}}
  }
  Lock_Top {
   widgetClassPath=&Image
   bIsVariable=true
   parent=Overlay_3
   widget=Lock_Top
   properties {
    visibility=HitTestInvisible
    brush {
     bIsDynamicallyLoaded=false
     drawAs=Image
     imageType=NoImage
     mirroring=NoMirror
     resourceName=None
     resourceObject=/SquadLockGameplay/UI/Icons/T_PadlockLock_Top
     tiling=NoTile
     imageSize=<180 180>
     margin=<ltrb 0 0 0 0>
     outlineSettings {
      bUseBrushTransparency=false
      roundingType=HalfHeightRadius
      width=0
      color {colorUseRule=UseColor_Specified, specifiedColor=<rgba 0 0 0 0>}
      cornerRadii=<0 0 0 0>
     }
     tintColor {colorUseRule=UseColor_Specified, specifiedColor=<rgba 1 1 1 1>}
     uVRegion {bIsValid=false, max=<0 0>, min=<0 0>}
    }
    renderTransform {angle=0, scale=<1 1>, shear=<0 0>, translation=<0 10>}
   }
   slot {class=OverlaySlot, properties {horizontalAlignment=HAlign_Center, verticalAlignment=VAlign_Center}}
  }
  Lock_Bottom {
   widgetClassPath=&Image
   bIsVariable=true
   parent=Overlay_3
   widget=Lock_Bottom
   properties {
    visibility=HitTestInvisible
    brush {
     bIsDynamicallyLoaded=false
     drawAs=Image
     imageType=NoImage
     mirroring=NoMirror
     resourceName=None
     resourceObject=/SquadLockGameplay/UI/Icons/T_PadlockLock_Bottom
     tiling=NoTile
     imageSize=<180 180>
     margin=<ltrb 0 0 0 0>
     outlineSettings {
      bUseBrushTransparency=false
      roundingType=HalfHeightRadius
      width=0
      color {colorUseRule=UseColor_Specified, specifiedColor=<rgba 0 0 0 0>}
      cornerRadii=<0 0 0 0>
     }
     tintColor {colorUseRule=UseColor_Specified, specifiedColor=<rgba 1 1 1 1>}
     uVRegion {bIsValid=false, max=<0 0>, min=<0 0>}
    }
   }
   slot {class=OverlaySlot, properties {horizontalAlignment=HAlign_Center, verticalAlignment=VAlign_Center}}
  }
 ]
}
